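# Material function that picks one of three material attribute layers per pixel from a coordinate value.
# The Coords input, defaulting to the red channel of UV channel 1, feeds a named reroute that drives nested If comparisons against fixed thresholds.
# These route Layer1, Layer2 or Layer3 straight to the output as a hard switch.
# An alternative BlendMaterialAttributes chain with the same threshold logic is present but not connected to the output.
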MaterialFunction /Game/Global/MaterialFunctions/Environment/MF_3LayerSort
graph {nodes [
  n1 = FunctionInput() {inputName=Layer1, inputType=FunctionInput_MaterialAttributes, sortPriority=1}
  n2 = FunctionInput() {inputName=Layer2, inputType=FunctionInput_MaterialAttributes, sortPriority=2}
  n3 = NamedRerouteUsage()
  n4 = Constant() {r=1}
  n5 = Constant()
  n6 = FunctionInput() {inputName=Layer3, inputType=FunctionInput_MaterialAttributes, sortPriority=3}
  BlendMaterialAttributes(
   A=BlendMaterialAttributes(A=@n1, B=@n2, Alpha=If(A=@n3, "A > B"=@n4, "A < B"=@n5) {constB=1}) {vertexAttributeBlendType=UseA}
   B=@n6
   Alpha=If(A=NamedRerouteUsage(), "A > B"=Constant() {r=1}, "A < B"=Constant()) {constB=2}
  ) {vertexAttributeBlendType=UseA}
  Constant()
  Constant() {r=1}
  Constant()
  Constant() {r=1}
  n7 = NamedRerouteUsage()
  FunctionOutput(None=If(A=NamedRerouteUsage(), "A > B"=@n6, "A < B"=If(A=@n7, "A > B"=@n2, "A < B"=@n1) {constB=1}) {constB=2})
  n8 = TextureCoordinate() {coordinateIndex=1}
  NamedRerouteDeclaration(None=FunctionInput(Preview=ComponentMask(None=@n8) {r=true}) {bUsePreviewValueAsDefault=true, inputName=Coords, inputType=FunctionInput_Vector2}) {name=CoordRoute, nodeColor=<rgba 0.2656 0 1 1>}
 ]}
