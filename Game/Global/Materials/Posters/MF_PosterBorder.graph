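# Material function that builds a poster border mask with a textured edge, exposed through Border-group parameters.
# Scaled, shifted UVs become a saturated box-edge gradient from the larger axis, tuned by size and fade; ShiftU drives both axes, leaving ShiftV unused.
# That gradient is V and repeated radial coordinates are U, wrapping a strip texture around the frame.
# The red channel, optionally inverted, is output, zeroed where the gradient is near zero.
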MaterialFunction /Game/Global/Materials/Posters/MF_PosterBorder
graph {nodes [
  n1 = ScalarParameter() {
   defaultValue=0.9
   desc=Stretches the border texture further into the frame leaving the start where it is
   group=Border
   parameterName=BorderEdgeFade
   sortPriority=4
  }
  n2 = FunctionInput(Preview=@n1) {bUsePreviewValueAsDefault=true, inputName=BorderEdgeFade, inputType=FunctionInput_Scalar, previewValue=<6 0 0 1>}
  n3 = FunctionInput(Preview=TextureCoordinate()) {bUsePreviewValueAsDefault=true, inputName=Coords, inputType=FunctionInput_Vector2}
  n4 = VectorParameter() {
   desc=Scale and Shift UV positions to move border into frame if not using the entire 0-1 space.
   group=Border
   parameterName=BorderUVs
   sortPriority=5
   channelNames {a=ShiftV, b=ShiftU, g=ScaleV, r=ScaleU}
   defaultValue=<rgba 1 1 0 0>
  }
  n5 = AppendVector(A=@n4.ShiftU, B=@n4.ShiftU)
  n6 = ConstantBiasScale(None=MaterialFunctionCall(UVs=Add(A=@n3, B=@n5), Texture Scale=ComponentMask(None=@n4) {g=true, r=true}) {materialFunction=&ScaleUVsByCenter}.UVs) {bias=-0.5, scale=2}
  n7 = ScalarParameter() {
   defaultValue=0.02
   desc=Moves the border start location further into the frame giving more solid border area.
   group=Border
   parameterName=BorderSize
   sortPriority=3
  }
  n8 = FunctionInput(Preview=@n7) {bUsePreviewValueAsDefault=true, inputName=BorderSize, inputType=FunctionInput_Scalar, previewValue=<6 0 0 1>}
  n9 = MaterialFunctionCall(Float2=Add(A=Abs(None=@n6), B=@n8)) {materialFunction=&BreakOutFloat2Components}
  n10 = Saturate(None=Max(A=InverseLinearInterpolate(A=@n2, Value=@n9.R) {constA=0.9}, B=InverseLinearInterpolate(A=@n2, Value=@n9.G) {constA=0.9}))
  n11 = ScalarParameter() {
   defaultValue=3
   desc=Repeats the border mask texture around the entire border length.
   group=Border
   parameterName=RepeatTexture
   sortPriority=2
  }
  n12 = FunctionInput(Preview=@n11) {bUsePreviewValueAsDefault=true, inputName=BorderEdgeScaleTexture, inputType=FunctionInput_Scalar, previewValue=<6 0 0 1>}
  n13 = Multiply(
   A=MaterialFunctionCall() {materialFunction=&VectorToRadialValue}.Radial Coordinates
   B=@n12
  ) {constB=6}
  n14 = TextureObjectParameter() {
   group=Border
   parameterName=BorderTexture
   samplerType=SAMPLERTYPE_LinearColor
   sortPriority=0
   texture=&/PosterEdgeTest01
  }
  n15 = TextureSample(
   UVs=AppendVector(A=ComponentMask(None=@n13) {r=true}, B=@n10)
   Tex=FunctionInput(Preview=@n14) {bUsePreviewValueAsDefault=true, inputName=BorderEdgeTexture, inputType=FunctionInput_Texture2D}
  ) {samplerType=SAMPLERTYPE_LinearColor}
  n16 = OneMinus(None=@n15.R)
  n17 = ScalarParameter() {
   desc="Invert the border texture mask.  Default ( 0 ) expects white on the bottom and black on top."
   group=Border
   parameterName=InvertTexture
   sortPriority=1
  }
  n18 = Round(None=FunctionInput(Preview=@n17) {bUsePreviewValueAsDefault=true, inputName=InvertTexture, inputType=FunctionInput_Scalar})
  n19 = Constant()
  FunctionOutput(None=If(A=@n10, "A > B"=LinearInterpolate(A=@n15.R, B=@n16, Alpha=@n18), "A == B"=@n19, "A < B"=@n19) {equalsThreshold=0.1})
 ]}
