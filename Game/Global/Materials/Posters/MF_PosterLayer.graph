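# Material function that builds a poster layer as material attributes from diffuse, normal, emissive and packed Specular/Metallic/Roughness inputs.
# Diffuse passes through desaturation and tint, while each unpacked SMR channel gets its own multiplier.
# Emissive is scaled by intensity, then blended toward a UV-based border falloff mask, built from powered distance to center on each axis.
# EyeAdaptationInverse then optionally compensates the glow for auto exposure, controlled by a canvas-layer parameter.
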MaterialFunction /Game/Global/Materials/Posters/MF_PosterLayer
graph {nodes [
  n1 = FunctionInput() {bUsePreviewValueAsDefault=true, inputName=Diffuse, previewValue=<0.5 0.5 0.5 1>}
  n2 = FunctionInput() {bUsePreviewValueAsDefault=true, inputName=DiffuseDesat, inputType=FunctionInput_Scalar, sortPriority=11, previewValue=<0 0 0 0>}
  n3 = FunctionInput() {bUsePreviewValueAsDefault=true, inputName=DiffuseMult, sortPriority=10, previewValue=<1 1 1 1>}
  n4 = FunctionInput() {bUsePreviewValueAsDefault=true, description="R=Specular, G=Metal, B=Rough", inputName=SMR, sortPriority=1, previewValue=<0.5 0 0.5 1>}
  n5 = MaterialFunctionCall(Float3=@n4) {materialFunction=&BreakOutFloat3Components}
  n6 = FunctionInput() {bUsePreviewValueAsDefault=true, inputName=MetallicMult, inputType=FunctionInput_Scalar, sortPriority=30, previewValue=<1 0 0 1>}
  n7 = FunctionInput() {bUsePreviewValueAsDefault=true, inputName=SpecularMult, inputType=FunctionInput_Scalar, sortPriority=20, previewValue=<1 0 0 1>}
  n8 = FunctionInput() {bUsePreviewValueAsDefault=true, inputName=RoughnessMult, inputType=FunctionInput_Scalar, sortPriority=40, previewValue=<1 0 0 1>}
  n9 = Multiply(A=FunctionInput() {inputName=EmissiveColor}, B=ScalarParameter() {group=1_CanvasLayer, parameterName=EmissiveIntensity})
  n10 = Abs(None=Add(A=TextureCoordinate()) {constB=-0.5})
  n11 = OneMinus(None=Power(
    Base=Multiply(A=@n10) {constB=2}
    Exp=ScalarParameter() {defaultValue=4, group=1_CanvasLayer, parameterName=EmissiveBorderRange, sortPriority=34}
   ))
  n12 = Multiply(A=@n9, B=Multiply(A=ComponentMask(None=@n11) {r=true}, B=ComponentMask(None=@n11) {g=true}))
  n13 = ScalarParameter() {group=1_CanvasLayer, parameterName=EmissiveBorderFade, sortPriority=33}
  n14 = SetMaterialAttributes!(
   MaterialAttributes
   BaseColor=Multiply(A=Desaturation(None=@n1, Fraction=@n2), B=@n3)
   Metallic=Multiply(A=@n5.G, B=@n6)
   Specular=Multiply(A=@n5.R, B=@n7)
   Roughness=Multiply(A=@n5.B, B=@n8)
   Normal=FunctionInput() {bUsePreviewValueAsDefault=true, inputName=Normal, sortPriority=2, previewValue=<0 0 1 1>}
   EmissiveColor=EyeAdaptationInverse(
    LightValueInput=LinearInterpolate(A=@n9, B=@n12, Alpha=@n13)
    AlphaInput=ScalarParameter() {group=1_CanvasLayer, parameterName=InverseEyeAdaptation, sortPriority=35}
   ).EyeAdaptationInverse
  ) {attributeSetTypes [
    #1
    #2
    #3
    #4
    #5
    #6
   ]}
  FunctionOutput(None=@n14)
 ]}
