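# M_PosterBase is a master material for posters and paintings, with optional frame and backing layers, border, grunge and LCD screen mode.
# MF_PosterLayer builds each layer, MF_3LayerSort combines them, and MF_FN_DestructibleDamage follows.
# Custom primitive data drives per-instance canvas transform, flipbook animation and layer cycling.
# LCD mode pixelates UVs, masks panel pixels and adds time-of-day, view-angle emissive with ray-tracing boost, plus randomized switching.
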Material /Game/Global/Materials/Posters/M_PosterBase
graph {
 properties {
  bAutomaticallySetUsageInEditor=false
  bUseMaterialAttributes=true
  bUsedWithInstancedStaticMeshes=true
  bUsedWithNanite=true
  pixelDepthOffsetMode=PDOM_Legacy
  previewMesh=/Game/Effects/Fort_Effects/Effects/Environment/Hera/Shards/Glow/SM_Env_TeaCake_Shard_Glows
  usedShadingModels=Default Lit
 }
 nodes [
  n1 = ObjectPositionWS()
  n2 = Constant3Vector() {constant=<rgba 5.4623 1.2346 9.2232 0>}
  n3 = Add(A=Time(), B=Fmod(A=DotProduct(A=@n1, B=@n2), B=Constant() {r=50814.0312}))
  Add(A=MaterialFunctionCall() {materialFunction=&LinearGradient}.UGradient, B=@n3)
  n4 = ScalarParameter() {defaultValue=793.7133}
  n5 = Panner(Coordinate=Multiply(A=NamedRerouteUsage()), Time=Floor(None=NamedRerouteUsage())) {speedX=0.1, speedY=0.1}
  n6 = Divide(A=NamedRerouteUsage()) {constB=0.3}
  n7 = Sine(None=Add(
    A=MaterialFunctionCall!(Frequency Coefficients, Wave Weights=@n4, In=@n5) {materialFunction=/Game/Environments/Asteria/Sets/NeonCity/Techart/Screen/MaterialFunctions/MF_Cheap1DNoise}.Noise_0_1
    B=Frac(None=@n6)
   )) {period=3}
  n8 = Round(None=Saturate(None=Add(A=@n7)))
  n9 = Saturate(None=MaterialFunctionCall(Wave Weights=Constant() {r=1000}) {materialFunction=&MF_Cheap1DNoise}.Noise_0_1)
  n10 = Sine(None=Divide(A=NamedRerouteUsage()) {constB=8})
  n11 = Multiply(A=Add(A=Multiply(A=@n10, B=@n10)) {constB=-1}) {constB=0.5}
  n12 = Fmod(A=@n10, B=Constant() {r=0.8})
  LinearInterpolate(B=Multiply(A=LinearInterpolate(B=@n8, Alpha=@n9)) {constB=0.01}, Alpha=Saturate(None=Multiply(A=@n11, B=@n12)))
  n13 = NamedRerouteUsage()
  n14 = GetMaterialAttributes(None=@n13) {attributeGetTypes [#1]}
  n15 = VectorParameter() {parameterName=BorderColor, defaultValue=<rgba 1 0.6005 0.2592 1>}
  n16 = MaterialFunctionCall!(BorderEdgeTexture, InvertTexture, BorderEdgeScaleTexture, BorderEdgeFade, BorderSize, Coords) {materialFunction=&/MF_PosterBorder}
  n17 = StaticSwitchParameter(
   True=SetMaterialAttributes!(MaterialAttributes=@n14.MaterialAttributes, Base Color=LinearInterpolate(A=@n14.Base Color, B=@n15.RGB, Alpha=@n16.Result)) {attributeSetTypes [#1]}
   False=@n13
  ) {group=Border, parameterName=UseBorder, sortPriority=-1}
  n18 = ScalarParameter() {
   defaultValue=1
   desc="The Backing for a painting needs less detail so we use a single texture.  This lets you swap it between layer 2 and 3 if your mesh was imported incorrectly."
   group=Global
   parameterName=BackingUseLayer3
  }
  n19 = NamedRerouteUsage()
  n20 = NamedRerouteUsage()
  n21 = StaticSwitchParameter(
   True=MaterialFunctionCall!(Coords, Layer1=@n17, Layer2=If(A=@n18, "A > B"=@n19, "A < B"=@n20) {constB=1}, Layer3=If(A=@n18, "A > B"=@n20, "A < B"=@n19) {constB=1}) {materialFunction=/Game/Global/MaterialFunctions/Environment/MF_3LayerSort}.Result
   False=@n17
  ) {
   defaultValue=true
   desc=False for simple posters but true is we need extra layers for backing and frame.
   group=Global
   parameterName=UseMultipleLayers
   sortPriority=-1
  }
  n22 = GetMaterialAttributes(None=@n21) {attributeGetTypes [#1, #2]}
  n23 = NamedRerouteUsage()
  n24 = Multiply(A=Multiply(A=@n22.Base Color, B=@n23), B=NamedRerouteUsage())
  n25 = Multiply(
   A=@n22.Emissive Color
   B=MaterialFunctionCall(
    Value_Morning=ScalarParameter() {defaultValue=15, group=4-LCD, parameterName=LCD Emissive Morning}
    Value_Day=ScalarParameter() {defaultValue=15, group=4-LCD, parameterName=LCD Emissive Day}
    Value_Evening=ScalarParameter() {defaultValue=15, group=4-LCD, parameterName=LCD Emissive Evening}
    Value_Night=ScalarParameter() {defaultValue=15, group=4-LCD, parameterName=LCD Emissive Night}
   ) {materialFunction=&Scalar_Time_of_Day}.Result
  ) {constB=15}
  n26 = Multiply(A=@n23, B=@n25)
  n27 = Multiply(A=@n25, B=ScalarParameter() {defaultValue=4, group=4-LCD, parameterName=LCD Lumen Boost})
  n28 = ScalarParameter() {group=4-LCD, parameterName=LCD Min Brightness}
  n29 = Power(
   Base=DotProduct(A=VertexNormalWS(), B=CameraVectorWS())
   Exp=ScalarParameter() {defaultValue=0.8, group=4-LCD, parameterName=LCD Fresnel Angle}
  )
  n30 = Multiply(A=Multiply(A=RayTracingQualitySwitch(Normal=@n26, RayTraced=@n27), B=LinearInterpolate(A=@n28, Alpha=@n29)), B=NamedRerouteUsage())
  MaterialFunctionCall(Material_In=StaticSwitchParameter(
    True=SetMaterialAttributes!(MaterialAttributes=@n22.MaterialAttributes, Base Color=@n24, Emissive Color=@n30) {attributeSetTypes [#1, #2]}
    False=@n21
   ) {group=4-LCD, parameterName=LCD Screen?, sortPriority=1}) {materialFunction=&MF_FN_DestructibleDamage}
  MaterialFunctionCall(Layer1=@n17, Layer2=NamedRerouteUsage(), Layer3=NamedRerouteUsage()) {materialFunction=&MF_3LayerSort}
  n31 = Multiply()
  n32 = Fresnel(ExponentIn=ScalarParameter() {defaultValue=0.5, group=4-LCD, parameterName=Emissive Fresnel Angle}, BaseReflectFractionIn=Constant())
  n33 = Saturate(None=LinearInterpolate(Alpha=@n32) {constA=1, constB=0.3})
  n34 = Multiply(
   A=MaterialFunctionCall!(Emissive=Multiply(A=@n31, B=@n33)) {materialFunction=/Game/Environments/Asteria/Sets/NeonCity/Techart/Screen/MaterialFunctions/MF_NeonCityEmissiveBalance}.Result
   B=NamedRerouteUsage()
  )
  n35 = MaterialFunctionCall() {materialFunction=&Color_Time_of_Day}
  n36 = Multiply(A=Power(Base=@n32) {constExponent=4}, B=LinearInterpolate(A=@n35.Result) {constAlpha=0.4})
  n37 = LinearInterpolate(A=Multiply(A=NamedRerouteUsage(), B=@n35.Result), Alpha=NamedRerouteUsage()) {constB=0}
  Multiply(
   A=Add(A=Add(A=@n34, B=@n36), B=Multiply(A=@n37) {constB=0.1})
   B=ScalarParameter() {defaultValue=1, group=4-LCD, parameterName=Emissive Intensity}
  ) {constB=0.5}
  n38 = ScalarParameter() {
   desc="Rotate the canvas.  This is for setting in an MIC and adds onto the CPD_CanvasRotation set in custom prim data."
   group=1_CanvasLayer
   parameterName=CanvasRotation
   primitiveDataIndex=5
  }
  NamedRerouteDeclaration(None=Add(A=ScalarParameter() {bUseCustomPrimitiveData=true, group=1_CanvasLayer, parameterName=CPD_CanvasRotation, primitiveDataIndex=5}, B=@n38)) {name=CanvasRot_Route, nodeColor=<rgba 1 0.3047 0 1>}
  n39 = ScalarParameter() {bUseCustomPrimitiveData=true, defaultValue=1, group=4-LCD, parameterName=CPD_NumRows, primitiveDataIndex=6}
  n40 = ScalarParameter() {bUseCustomPrimitiveData=true, defaultValue=1, group=4-LCD, parameterName=CPD_NumColumns, primitiveDataIndex=12}
  n41 = Multiply(A=Time(), B=ScalarParameter() {bUseCustomPrimitiveData=true, group=4-LCD, parameterName=CPD_FlipbookAnimSpeed, primitiveDataIndex=16})
  n42 = TextureCoordinate()
  n43 = VectorParameter() {
   group=1_CanvasLayer
   parameterName=CanvasUVs
   sortPriority=3
   channelNames {a=ShiftV, b=ShiftU, g=ScaleV, r=ScaleU}
   defaultValue=<rgba 1 1 0 0>
  }
  n44 = Add(
   A=AppendVector(A=@n43.ShiftU, B=@n43.ShiftU)
   B=AppendVector(
    A=ScalarParameter() {bUseCustomPrimitiveData=true, parameterName=CPD_CanvasShiftU, primitiveDataIndex=3}
    B=ScalarParameter() {bUseCustomPrimitiveData=true, parameterName=CPD_CanvasShiftV, primitiveDataIndex=4}
   )
  )
  n45 = ScalarParameter() {bUseCustomPrimitiveData=true, parameterName=CPD_CanvasGlobalScale}
  n46 = Multiply(A=AppendVector(A=@n43.ScaleU, B=@n43.ScaleU), B=If(A=@n45, "A > B"=@n45, "A == B"=Constant() {r=1}, "A < B"=@n45))
  n47 = ScalarParameter() {bUseCustomPrimitiveData=true, parameterName=CPD_CanvasScaleU, primitiveDataIndex=1}
  n48 = ScalarParameter() {bUseCustomPrimitiveData=true, parameterName=CPD_CanvasScaleV, primitiveDataIndex=2}
  n49 = AppendVector(A=If(A=@n47, "A > B"=@n47, "A == B"=Constant() {r=1}, "A < B"=@n47), B=If(A=@n48, "A > B"=@n48, "A == B"=Constant() {r=1}, "A < B"=@n48))
  n50 = MaterialFunctionCall(
   UVs=MaterialFunctionCall(UVs=Add(A=@n42, B=@n44), Texture Scale=Multiply(A=@n46, B=@n49)) {materialFunction=&ScaleUVsByCenter}.UVs
   "Rotation Angle (0-1)"=NamedRerouteUsage()
  ) {materialFunction=&CustomRotator}
  n51 = NamedRerouteDeclaration(None=If(
    A=Max(A=@n39, B=@n40)
    "A > B"=MaterialFunctionCall("Animation  Phase (0-1)"=@n41, Number of Rows=@n39, Number of Columns=@n40) {materialFunction=&FlipBook}.UVs
    "A < B"=VertexInterpolator(VS=@n50.Rotated Values).PS
   ) {constB=1.1}) {name=Base UVs, nodeColor=<rgba 1 0 0.2578 1>}
  n52 = ScalarParameter() {defaultValue=275, group=4-LCD, parameterName=LCDTiling}
  n53 = Multiply(A=@n51, B=@n52)
  NamedRerouteDeclaration(None=StaticSwitchParameter(
    True=TextureSample(UVs=@n53) {samplerType=SAMPLERTYPE_Masks, texture=/Game/Environments/Asteria/Sets/NeonCity/Techart/Textures/T_LCD_Panels_LCDSoft}.RGB
    False=Constant() {r=1}
   ) {group=4-LCD, parameterName=LCD Screen?, sortPriority=1}) {name=Pixels, nodeColor=<rgba 1 0.7734 0 1>}
  n54 = Divide(A=Ceil(None=@n53), B=@n52)
  n55 = NamedRerouteDeclaration(None=StaticSwitchParameter(True=NamedRerouteDeclaration(None=@n54) {name=Pixelated UVs, nodeColor=<rgba 0.1953 1 0 1>}, False=@n51) {group=4-LCD, parameterName=LCD Screen?, sortPriority=1}) {name=UVs, nodeColor=<rgba 0.7813 0 1 1>}
  n56 = TextureSampleParameter2D(UVs=@n55) {
   group=1_CanvasLayer
   parameterName=SpecularMasks
   samplerType=SAMPLERTYPE_Masks
   sortPriority=1
   texture=/Game/Global/Textures/Default/Blanks/T_Blank_Mask_Magenta_R1G0B1A1
   channelNames {a=---, b=Roughness, g=Metallic, r=Specular}
  }
  NamedRerouteDeclaration(None=@n56.RGB) {name=Spec Mask, nodeColor=<rgba 1 0 0.6562 1>}
  NamedRerouteDeclaration() {name=Glitches, nodeColor=<rgba 0 0.1328 1 1>}
  n57 = Round(None=ScalarParameter() {bUseCustomPrimitiveData=true, parameterName=Layer Cycling?, primitiveDataIndex=11, sliderMax=1})
  n58 = Sine(None=Multiply(
    A=NamedRerouteDeclaration(None=@n3) {name=perACtorRand, nodeColor=<rgba 0 0.9531 1 1>}
    B=ScalarParameter() {defaultValue=0.01, group=4-LCD, parameterName=Switch Speed}
   ) {constB=0.3})
  n59 = Add(A=Round(None=@n58))
  n60 = Saturate(None=Multiply(A=Saturate(None=@n57), B=Multiply(A=@n59) {constB=0.5}))
  NamedRerouteDeclaration(None=StaticSwitchParameter(True=VertexInterpolator(VS=@n60).PS, False=Constant()) {group=4-LCD, parameterName=LCD Screen?, sortPriority=1}) {name=Switch, nodeColor=<rgba 0 0.2734 1 1>}
  n61 = Saturate(None=Sine(None=Multiply(A=@n58, B=@n58)))
  n62 = Constant() {r=0.5}
  n63 = MaterialFunctionCall(In=Power(Base=Multiply(A=@n61, B=@n62), Exp=Constant() {r=4}) {constExponent=12}, Contrast=Constant() {r=0.04}) {materialFunction=&CheapContrast}
  n64 = Constant() {r=43595}
  n65 = Saturate(None=SmoothStep(Min=Constant() {r=28}, Max=Constant() {r=1084}, Value=Multiply(A=@n63.Result, B=@n64) {constB=600}))
  n66 = Round(None=ScalarParameter() {bUseCustomPrimitiveData=true, parameterName=Layer Cycling?, primitiveDataIndex=11, sliderMax=1})
  NamedRerouteDeclaration(None=LinearInterpolate(B=OneMinus(None=@n65), Alpha=Saturate(None=@n66)) {constA=1}) {name=Transition Fade, nodeColor=<rgba 0 0.0391 1 1>}
  n67 = Saturate(None=ScalarParameter() {defaultValue=1, group=2_CanvasGrunge, parameterName=Grunge_InvLerp_A})
  n68 = ScalarParameter() {defaultValue=-0.5, group=2_CanvasGrunge, parameterName=Grunge_InvLerp_B}
  n69 = TextureObjectParameter() {group=2_CanvasGrunge, parameterName=GrungeMask, samplerType=SAMPLERTYPE_Masks, texture=&/T_Leakage_Streaks_02_D}
  n70 = ChannelMaskParameter(None=MaterialFunctionCall(TextureObject=@n69, TextureSize=ScalarParameter() {defaultValue=200, group=2_CanvasGrunge, parameterName=GrungTileAmount}) {materialFunction=&WorldAlignedTexture}.XYZ Texture) {group=2_CanvasGrunge, parameterName=GrungeMaskChannel}
  n71 = StaticSwitchParameter(True=OneMinus(None=@n70.R), False=@n70.R) {defaultValue=true, group=2_CanvasGrunge, parameterName=InvertGrungeMask}
  n72 = StaticSwitchParameter(True=Constant() {r=0.5}, False=Saturate(None=InverseLinearInterpolate(A=@n67, B=@n68, Value=@n71) {constA=-0.25, constB=0.85})) {group=4-LCD, parameterName=LCD Screen?, sortPriority=1}
  n73 = TextureSampleParameter2D(UVs=@n55) {group=1_CanvasLayer, parameterName=Diffuse, sortPriority=0, texture=/Game/Global/Textures/Default/Blanks/T_FlatGrey}
  n74 = Desaturation(
   None=MaterialFunctionCall(
    A=NamedRerouteDeclaration(None=@n73.RGB) {name=Dif 1, nodeColor=<rgba 0.5937 1 0 1>}
    B=NamedRerouteUsage()
    C=NamedRerouteUsage()
    Alpha=NamedRerouteUsage()
   ) {materialFunction=&Lerp_3Color}.Result
   Fraction=ScalarParameter() {group=1_CanvasLayer, parameterName=DesaturateDiffuse}
  )
  n75 = StaticSwitchParameter(
   True=LinearInterpolate(A=@n74, B=VectorParameter() {group=2_CanvasGrunge, parameterName=GrungeColor, defaultValue=<rgba 0.0375 0.0469 0.009 1>}.RGB, Alpha=@n72)
   False=@n74
  ) {group=2_CanvasGrunge, parameterName=UseGrunge}
  n76 = RayTracingQualitySwitch(Normal=@n74, RayTraced=Multiply(A=@n74, B=ScalarParameter() {defaultValue=1, group=1_CanvasLayer, parameterName=Emissive Lumen Multiplier}))
  n77 = StaticSwitchParameter(True=LinearInterpolate(A=@n76, Alpha=@n72) {constB=0}, False=@n76) {group=2_CanvasGrunge, parameterName=UseGrunge}
  n78 = TextureSampleParameter2D(UVs=@n55) {
   group=1_CanvasLayer
   parameterName=Normals
   samplerType=SAMPLERTYPE_Normal
   sortPriority=2
   texture=&FlatNormalBlue
  }
  n79 = StaticSwitchParameter(True=Constant3Vector() {constant=<rgba 0 0 1 0>}, False=@n78.RGB) {group=4-LCD, parameterName=LCD Screen?, sortPriority=1}
  n80 = VectorParameter() {
   desc=RGB is color multiplier but Alpha is a desaturation to the original color before the multiply so you can remove color before tinting.
   group=1_CanvasLayer
   parameterName=CanvasDiffuseMult
   sortPriority=4
   channelNames {a=Pre-Desat, b=Blue, g=Green, r=Red}
   defaultValue=<rgba 1 1 1 0>
  }
  n81 = VectorParameter() {
   group=1_CanvasLayer
   parameterName=CanvasMaskMult
   sortPriority=5
   channelNames {a=---, b=Roughness, g=Metallic, r=Specular}
   defaultValue=<rgba 1 1 1 1>
  }
  NamedRerouteDeclaration(None=SetMaterialAttributes!(MaterialAttributes=MaterialFunctionCall!(
     Diffuse=@n75
     EmissiveColor=@n77
     SMR=@n56.RGB
     Normal=@n79
     DiffuseMult=@n80
     DiffuseDesat=@n80
     SpecularMult=@n81.Specular
     MetallicMult=@n81.Specular
     RoughnessMult=@n81.Specular
    ) {materialFunction=&/MF_PosterLayer}.Result)) {name=Layer1Route, nodeColor=<rgba 0 0.6953 1 1>}
  n82 = NamedRerouteUsage()
  n83 = TextureSampleParameter2D(UVs=@n82) {group=2_FrameLayer, parameterName=Diffuse_Texture_2, sortPriority=0, texture=&T_FlatGrey}
  n84 = MaterialFunctionCall(
   A=NamedRerouteDeclaration(None=@n83.RGB) {name=Dif 2, nodeColor=<rgba 1 0 0.9141 1>}
   B=NamedRerouteUsage()
   C=NamedRerouteUsage()
   Alpha=NamedRerouteUsage()
  ) {materialFunction=&Lerp_3Color}
  n85 = TextureSampleParameter2D(UVs=@n82) {
   group=2_FrameLayer
   parameterName=SpecularMasks_2
   samplerType=SAMPLERTYPE_Masks
   sortPriority=0
   texture=&T_Blank_Mask_Magenta_R1G0B1A1
   channelNames {a=---, b=Roughness, g=Metallic, r=Specular}
  }
  n86 = StaticSwitchParameter(True=NamedRerouteUsage(), False=@n85.RGB) {group=4-LCD, parameterName=LCD Screen?, sortPriority=1}
  n87 = TextureSampleParameter2D(UVs=@n82) {
   group=2_FrameLayer
   parameterName=Normals_Texture_2
   samplerType=SAMPLERTYPE_Normal
   sortPriority=0
   texture=/Engine/EngineMaterials/FlatNormal
  }
  n88 = StaticSwitchParameter(True=Constant3Vector() {constant=<rgba 0 0 1 0>}, False=@n87.RGB) {group=4-LCD, parameterName=LCD Screen?, sortPriority=1}
  NamedRerouteDeclaration(None=SetMaterialAttributes(MaterialAttributes=MaterialFunctionCall(Diffuse=@n84.Result, EmissiveColor=@n84.Result, SMR=@n86, Normal=@n88) {materialFunction=&MF_PosterLayer}.Result)) {name=Layer2Route, nodeColor=<rgba 0 0.6953 1 1>}
  n89 = TextureSampleParameter2D(UVs=NamedRerouteUsage()) {group=3_BackingLayer, parameterName=Diffuse_Texture_3, sortPriority=0, texture=&T_FlatGrey}
  n90 = MaterialFunctionCall(
   A=NamedRerouteDeclaration(None=@n89.RGB) {name=Dif 3, nodeColor=<rgba 0 1 0.7656 1>}
   B=NamedRerouteUsage()
   C=NamedRerouteUsage()
   Alpha=NamedRerouteUsage()
  ) {materialFunction=&Lerp_3Color}
  n91 = VectorParameter() {
   group=3_BackingLayer
   parameterName=BackingSMR
   sortPriority=5
   channelNames {a=---, b=Roughness, g=Metallic, r=Specular}
   defaultValue=<rgba 0.5 0 0.5 1>
  }
  NamedRerouteDeclaration(None=SetMaterialAttributes(MaterialAttributes=MaterialFunctionCall(Diffuse=@n90.Result, EmissiveColor=@n90.Result, SMR=@n91) {materialFunction=&MF_PosterLayer}.Result)) {name=Layer3Route, nodeColor=<rgba 0.6172 1 0 1>}
  n92 = TextureCoordinate()
  Saturate(None=InverseLinearInterpolate(Value=ComponentMask(None=@n92) {r=true}) {constA=0.25, constB=0.75})
  Time()
  Time()
 ]
}
